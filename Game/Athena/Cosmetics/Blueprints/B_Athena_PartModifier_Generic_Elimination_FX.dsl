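; Generic cosmetic part modifier that plays elimination VFX on a character when its tracked stat rises above zero.
; Each change calls the empty OnReactivity hook, replaces the previous elimination particle, and feeds it kill count, timestamp, frontend and hidden-part flags.
; Without elimination spawn parameters, it updates the tagged idle VFX instead.
; When frontend reactivity applies, post-res-in fakes a kill after a delay to preview the effect.
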(event EventOnPostResIn
  (bind _returnvalue (CallFunction|GetVFXComponent "(TagName=\"CosmeticPart.VFX.Idle\")"))
  (|CallParentFunction)
  (if (CallFunction|ShouldDisplayFrontendReactivity)
    (Utilities|FlowControl|Delay (Variables|Default|GetFEDelayDuration))
    (bind _result (Math|Integer|IncrementInt (Variables|Default|GetFEKillCount))
      (:%20%20
        (CallFunction|OnReactivity _result)
        (bind _returnvalue_2 (CallFunction|GetVFXSpawnParameters "(TagName=\"CosmeticPart.VFX.Elimination\")")
          (:Success
            (bind _returnvalue_3 (PartModifier|SpawnSystem|SpawnSystemAttachedPartModifier (Utilities|Struct|BreakCharacterPartModifierVFXSpawnParameters _returnvalue_2) (.AttachToComponent (Utilities|Struct|BreakCharacterPartModifierVFXSpawnParameters _returnvalue_2)) (.SocketName (Utilities|Struct|BreakCharacterPartModifierVFXSpawnParameters _returnvalue_2)) "0, 0, 0" "0, 0, 0" (.LocationType (Utilities|Struct|BreakCharacterPartModifierVFXSpawnParameters _returnvalue_2)) true (.bAutoActivate (Utilities|Struct|BreakCharacterPartModifierVFXSpawnParameters _returnvalue_2)))
              (:Success
                (bind _output_get (Variables|Default|SetEliminationParticle _returnvalue_3))
                (Niagara|SetNiagaraVariable(Bool) _output_get "bisFrontEnd" (Reactive|Frontend|IsCosmeticObjectDisplayedinFrontendContext 0))
                (Effects|Components|ParticleSystem|SetFloatParameter _output_get "TimeStamp" (Utilities|Time|GetGameTimeinSeconds))
                (Utilities|FlowControl|Delay 8.0))))
          (:Failure
            _))
        (Effects|Components|ParticleSystem|SetIntParameter _returnvalue "KillCount" _result)
        (Effects|Components|ParticleSystem|SetFloatParameter _returnvalue "TimeStamp" (Utilities|Time|GetGameTimeinSeconds))
        (Utilities|FlowControl|Delay 8.0)))))

(event EventOnStatChanged
  (params (State%20Name :type Name :default "None") (Stat%20Value :type Integer))
  (bind _returnvalue (CallFunction|GetVFXComponent "(TagName=\"CosmeticPart.VFX.Idle\")"))
  (|CallParentFunction State%20Name Stat%20Value)
  (if (> Stat%20Value 0)
    (CallFunction|OnReactivity Stat%20Value)
    (bind _returnvalue_2 (CallFunction|GetVFXSpawnParameters "(TagName=\"CosmeticPart.VFX.Elimination\")")
      (:Success
        ; Disables a particle to avoid stacking after getting multiple elims (Set the NS to Kill in order to make this works
        (Components|Activation|Deactivate (Variables|Default|GetEliminationParticle))
        (bind _returnvalue_3 (PartModifier|SpawnSystem|SpawnSystemAttachedPartModifier (Utilities|Struct|BreakCharacterPartModifierVFXSpawnParameters _returnvalue_2) (.AttachToComponent (Utilities|Struct|BreakCharacterPartModifierVFXSpawnParameters _returnvalue_2)) (.SocketName (Utilities|Struct|BreakCharacterPartModifierVFXSpawnParameters _returnvalue_2)) "0, 0, 0" "0, 0, 0" (.LocationType (Utilities|Struct|BreakCharacterPartModifierVFXSpawnParameters _returnvalue_2)) true (.bAutoActivate (Utilities|Struct|BreakCharacterPartModifierVFXSpawnParameters _returnvalue_2)))
          (:Success
            (bind _output_get (Variables|Default|SetEliminationParticle _returnvalue_3))
            (Niagara|SetNiagaraVariable(Bool) _output_get "bisFrontEnd" (Reactive|Frontend|IsCosmeticObjectDisplayedinFrontendContext 0))
            (Niagara|SetNiagaraVariable(Int32) _output_get "KillCount" Stat%20Value)
            (Effects|Components|ParticleSystem|SetFloatParameter _output_get "TimeStamp" (Utilities|Time|GetGameTimeinSeconds))
            (CallFunction|SetSpawnedParticleHiddenTags _returnvalue_3))))
      (:Failure
        (Effects|Components|ParticleSystem|SetIntParameter _returnvalue "KillCount" Stat%20Value)
        (Effects|Components|ParticleSystem|SetFloatParameter _returnvalue "TimeStamp" (Utilities|Time|GetGameTimeinSeconds))))))

(event Custom|OnReactivity
  (params (KillCount :type Integer)))

(fn SetSpawnedParticleHiddenTags
  (params (NiagaraParticle :type Niagara%20Particle%20System%20Component%20Object%20Reference))
  (Niagara|SetNiagaraVariable(Bool) NiagaraParticle "bisHeadHidden" (Variables|HiddenPartCheck|GetbIsHeadHidden))
  (Niagara|SetNiagaraVariable(Bool) NiagaraParticle "bisHandHidden" (Variables|HiddenPartCheck|GetbIsHandHidden))
  (Niagara|SetNiagaraVariable(Bool) NiagaraParticle "bisFootHidden" (Variables|HiddenPartCheck|GetbIsFootHidden)))

(fn ConstructionScript
  (|CallParentFunction))
