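; Generic cosmetic part modifier Blueprint for Athena that manages its own actor tick and provides velocity helper functions.
; Swim, skydive and slide begin events set tick from per-state toggles; their end events and post-res-in re-enable it, the latter also printing a debug string.
; Helpers normalize speed against sprint and walk thresholds, and build a clamped, decaying accumulated velocity using separate walk and sprint multipliers.
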(event EventOnBeginSwimming
  ;[1 1 1] Comment
  (|CallParentFunction)
  (Actor|Tick|SetActorTickEnabled (Variables|TickToggles|GetTickOnSwim)))

(event EventOnPostResIn
  (|CallParentFunction)
  (Development|PrintString (Utilities|String|Append "Enabling Tick for: " (Utilities|GetDisplayName self)) false true "(R=1.000000,G=0.455398,B=0.000000,A=1.000000)")
  (Actor|Tick|SetActorTickEnabled true))

(event EventOnEndSwimming
  ;[1 1 1] Comment
  (|CallParentFunction)
  (Actor|Tick|SetActorTickEnabled true))

(event EventOnBeginSkydiving
  ;[1 1 1] Comment
  (|CallParentFunction)
  (Actor|Tick|SetActorTickEnabled (Variables|TickToggles|GetTickOnSkydive)))

(event EventOnEndSkydiving
  ;[1 1 1] Comment
  (|CallParentFunction)
  (Actor|Tick|SetActorTickEnabled true))

(event EventOnBeginSliding
  ;[1 1 1] Comment
  (|CallParentFunction)
  (Actor|Tick|SetActorTickEnabled (Variables|TickToggles|GetTickOnSliding)))

(event EventOnEndSliding
  ;[1 1 1] Comment
  (|CallParentFunction)
  (Actor|Tick|SetActorTickEnabled true))

(fn GetNormalizedAccumulatedVelocity
  (params (DeltaSeconds :type Float%20%28double-precision%29%20%28by%20ref%29) (LastAccumulatedVelocity :type Float%20%28double-precision%29%20%28by%20ref%29) (MaxAccumulatedVelocity :type Float%20%28double-precision%29 :default 5e+04) (WalkAccumMultiplier :type Float%20%28double-precision%29 :default 0.75) (SprintAccumMultiplier :type Float%20%28double-precision%29 :default 2.25) (DecaySpeed :type Float%20%28double-precision%29 :default 0.3) (Velocity :type Vector))
  (LocalVariables|SetVelocityVectorLength (Math|Vector|VectorLength Velocity))
  (return (* (Math|Interpolation|FInterpTo (Math|Float|Clamp(Float) (+ (Math|Float|Lerp (* (LocalVariables|GetVelocityVectorLength) (Variables|GetWalkAccumMultiplier)) (* (LocalVariables|GetVelocityVectorLength) (Variables|GetSprintAccumMultiplier)) (Math|Conversions|ToFloat(Boolean) (> (LocalVariables|GetVelocityVectorLength) (LocalVariables|GetSprintTrigger)))) (Variables|GetLastAccumulatedVelocity)) 0.0 (Variables|GetMaxAccumulatedVelocity)) 0.0 (Variables|GetDeltaSeconds) (Variables|GetDecaySpeed)) (/ 1.0 (Variables|GetMaxAccumulatedVelocity))) (Math|Interpolation|FInterpTo (Math|Float|Clamp(Float) (+ (Math|Float|Lerp (* (LocalVariables|GetVelocityVectorLength) (Variables|GetWalkAccumMultiplier)) (* (LocalVariables|GetVelocityVectorLength) (Variables|GetSprintAccumMultiplier)) (Math|Conversions|ToFloat(Boolean) (> (LocalVariables|GetVelocityVectorLength) (LocalVariables|GetSprintTrigger)))) (Variables|GetLastAccumulatedVelocity)) 0.0 (Variables|GetMaxAccumulatedVelocity)) 0.0 (Variables|GetDeltaSeconds) (Variables|GetDecaySpeed))))

(fn GetNormalizedVelocity
  (params (Velocity :type Vector))
  (LocalVariables|SetVelocityVectorLength (Math|Vector|VectorLength Velocity))
  (return (Math|Float|Clamp(Float) (* (/ 1.0 (LocalVariables|GetSprintTrigger)) (LocalVariables|GetVelocityVectorLength))) (Math|Float|Clamp(Float) (* (/ 1.0 (LocalVariables|GetMaxWalkSpeed)) (LocalVariables|GetVelocityVectorLength))) (LocalVariables|GetVelocityVectorLength)))

(fn ConstructionScript
  (|CallParentFunction))
