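; Parent cosmetic part modifier that turns the associated player pawn's movement into overridable events for child Blueprints.
; Each tick it stores pawn velocity, optionally zeroing Z, then runs flag-gated branches for normalized linear velocity, decaying accumulated velocity and a moving/stopped boolean.
; A flag-gated listener compares the pawn's tactical sprint state to a cached value and fires an event only on change.
; All event handlers here are empty.
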(event EventOnPostResIn
  (|CallParentFunction))

(event EventTick
  (params (DeltaSeconds :type Float%20%28single-precision%29))
  (|CallParentFunction DeltaSeconds)
  (Utilities|FlowControl|Sequence
    (:then_0
      ;[1 1 1] GetCharacterVelocity
      (Variables|Velocity|SetPawnVelocity (Math|Vector|SelectVector (* (Transformation|GetVelocity (CustomCharacter|GetAssociatedPlayerPawn)) "1.000000,1.000000,0.000000") (Transformation|GetVelocity (CustomCharacter|GetAssociatedPlayerPawn)) (Variables|Default-Velocity|GetAllowVerticalVelocity))))
    (:then_1
      ;[0.3166 1 0.4028] Example 1 - Simple Velocity
      (if (Variables|Default-Velocity|GetUseLinearVelocity)
        (bind (_normalizedoverallvelocity _normalizedwalkvelocity _velocitylength) (Velocity|GetNormalizedVelocity (Variables|Velocity|GetPawnVelocity)))
        (CallFunction|OnLinearVelocity _normalizedoverallvelocity _normalizedwalkvelocity _velocitylength)))
    (:then_2
      ;[1 0.3372 0.6707] Example 2 - Accumulated Velocity
      (if (Variables|Default-Velocity|GetUseAccumulatedVelocity)
        (bind (_normalized_accumulatedvelocity _accumulated_velocity) (Velocity|GetNormalizedAccumulatedVelocity (Utilities|Time|GetWorldDeltaSeconds) (stale _output_get) 5e+04 0.75 2.25 (Variables|Velocity|GetDecaySpeed) (Variables|Velocity|GetPawnVelocity)))
        (bind _output_get (Variables|Velocity|SetAccumulated%20Velocity _accumulated_velocity))
        (CallFunction|OnAccumulatedVelocity _normalized_accumulatedvelocity _output_get)))
    (:then_3
      ;[0.2353 0.5182 1] Example 3 - Simple Toggle
      (if (Variables|Default-Velocity|GetUseToggleVelocity)
        (CallFunction|OnVelocityToggle (Math|Conversions|ToBoolean(Integer) (Math|Integer|Clamp(Integer) (Math|Float|Round (Math|Vector|VectorLength (Variables|Velocity|GetPawnVelocity))) 0 1)) DeltaSeconds)))
    (:then_4
      ; TacticalSprinting
      (if (Variables|Default-Velocity|GetUseTacticalSprintingEvents)
        (CallFunction|TacticalSprintingListhener)))))

(event Custom|OnTacticalSprintingToggle
  (params (Is%20TacticalSprinting :type Boolean)))

(event Custom|OnLinearVelocity
  (params (NormalizedOverallVelocity :type Float%20%28double-precision%29) (NormalizedWalkVelocity :type Float%20%28double-precision%29) (VelocityLength :type Float%20%28double-precision%29)))

(event Custom|OnAccumulatedVelocity
  (params (NormalizedAccumulatedVelocity :type Float%20%28double-precision%29) (AccumulatedVelocity :type Float%20%28double-precision%29)))

(event Custom|OnVelocityToggle
  (params (IsMoving :type Boolean) (Delta%20Seconds :type Float%20%28double-precision%29)))

(fn TacticalSprintingListhener
  (bind _returnvalue (CustomCharacter|GetAssociatedPlayerPawn))
  (bind _bistacticalsprinting (Variables|/Script/FortniteGame.FortPlayerPawn|GetbIsTacticalSprinting _returnvalue))
  (Utilities|FlowControl|Sequence
    (:then_0
      (if (not (== _bistacticalsprinting (Variables|CharacterStates|GetIs%20Tactical%20Sprinting)))
        (CallFunction|OnTacticalSprintingToggle _bistacticalsprinting)))
    (:then_1
      (Variables|CharacterStates|SetIs%20Tactical%20Sprinting _bistacticalsprinting))))

(fn ConstructionScript
  (|CallParentFunction))
